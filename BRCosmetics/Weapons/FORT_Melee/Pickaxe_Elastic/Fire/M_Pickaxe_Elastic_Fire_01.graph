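# Translucent unlit Niagara mesh-particle material for the Elastic pickaxe fire effect.
# A curve atlas along U blends two Dynamic Parameter colors (both reuse red for green), dimming backfaces, modulated by panning noise with optional glitch tint.
# Opacity combines view-angle and depth fades, with optional particle-color shaping and scanlines.
# A panning texture, masked toward V edges, drives world-position-offset wobble; the Primary/Secondary Color vector parameters are unconnected.
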Material /BRCosmetics/Weapons/FORT_Melee/Pickaxe_Elastic/Fire/M_Pickaxe_Elastic_Fire_01
graph {
 properties {
  bUsedWithNiagaraMeshParticles=true
  blendMode=BLEND_Translucent
  pixelDepthOffsetMode=PDOM_Legacy
  shadingModel=MSM_Unlit
  usedShadingModels=Unlit
 }
 nodes [
  n1 = ParticleColor()
  n2 = DepthFade(
   Opacity=MaterialFunctionCall("FadeSensitivity (optional)"=ScalarParameter() {defaultValue=2, parameterName=AngleFade}) {materialFunction=&FadeBasedOnViewAngle}.Result
   FadeDistance=ScalarParameter() {defaultValue=16, parameterName=FadeDistance}
  )
  n3 = Multiply(A=@n1.A, B=@n2)
  n4 = LinearInterpolate(
   A=Constant() {r=0.2}
   Alpha=TextureSample() {samplerType=SAMPLERTYPE_Grayscale, texture=&LowResBlurredNoise}.R
  ) {constA=0.2}
  n5 = ComponentMask(None=TextureCoordinate()) {g=true}
  n6 = ScalarParameter() {parameterName=Use_ParticleColor}
  n7 = LinearInterpolate(A=@n2, B=Multiply(A=Multiply(A=@n3, B=@n4), B=Power(Base=@n5) {constExponent=1.5}), Alpha=@n6)
  n8 = ScalarParameter() {defaultValue=50, group=Glitch, parameterName=Scanline Amount}
  n9 = ScalarParameter() {defaultValue=-0.15, group=Glitch, parameterName=ScanlineSize}
  n10 = ScalarParameter() {defaultValue=0.5, group=Glitch, parameterName=Scanlines Dark}
  n11 = ScalarParameter() {defaultValue=1, group=Glitch, parameterName=Scanlines Light}
  LinearInterpolate(
   A=@n7
   B=Multiply(
    A=@n7
    B=MaterialFunctionCall!(
     Number of Scanlines=@n8
     Scanline Speed
     "Scanlines Thickness (Bias)"=@n9
     Scanlines Dark Value=@n10
     Scanlines Light Value=@n11
     Block Opacity
     UVs
     Time
     Block Repeat
    ) {materialFunction=/Game/Effects/Fort_Effects/Effects/Atlas/Materials/MF_Atlas_Scanlines_01}.Result
   )
   Alpha=ScalarParameter() {group=Glitch, parameterName=Use_Scanlines}
  )
  n12 = CurveAtlasRowParameter(CurveTime=ComponentMask(None=TextureCoordinate()) {r=true}) {
   atlas=/Game/Athena/LinearColorCurves/TempAthenaCurveAtlas
   curve=/Game/Weapons/FORT_Melee/Pickaxe_Elastic/Fire/Curve_Pickaxe_Elastic_Fire_01
   defaultValue=14
  }
  n13 = DynamicParameter() {parameterIndex=1, paramNames [SecondaryR,SecondaryG,SecondaryB,SecondaryA]}
  n14 = AppendVector(A=AppendVector(A=AppendVector(A=@n13.SecondaryR, B=@n13.SecondaryR), B=@n13.SecondaryB), B=@n13.SecondaryA)
  n15 = Desaturation(
   None=Multiply(A=@n14, B=ScalarParameter() {defaultValue=0.08, group=Color, parameterName=Backside Color Intensity})
   Fraction=ScalarParameter() {defaultValue=0.3, group=Color, parameterName=Backside Color Desaturation}
  )
  n16 = TextureCoordinate()
  n17 = AppendVector(
   A=ScalarParameter() {defaultValue=1.5, group=Noise, parameterName=NoiseTexture_Tile_U}
   B=ScalarParameter() {defaultValue=0.5, group=Noise, parameterName=NoiseTexture_Tile_V}
  )
  n18 = ScalarParameter() {group=Noise, parameterName=NoiseTexture_Pan_U}
  n19 = ScalarParameter() {group=Noise, parameterName=NoiseTexture_Pan_V}
  n20 = TextureSampleParameter2D(UVs=Panner(Coordinate=Multiply(A=@n16, B=@n17), Speed=AppendVector(A=@n18, B=@n19)) {speedY=0.3}) {group=Noise, parameterName=NoiseTexture, texture=/Game/Packages/Fortress_Transylvania/Haunted_Forest/Swamps/Textures/2ndInfectedMapLinear2}
  n21 = Multiply(
   A=Power(Base=@n20.R, Exp=ScalarParameter() {defaultValue=1, group=Noise, parameterName=NoiseTexture_Power})
   B=ScalarParameter() {defaultValue=1, group=Noise, parameterName=NoiseTexture_Intensity}
  )
  n22 = ScalarParameter() {defaultValue=1, group=Color, parameterName=Secondary Color Intensity}
  n23 = DynamicParameter() {paramNames [PrimaryR,PrimaryG,PrimaryB,PrimaryA]}
  n24 = AppendVector(A=AppendVector(A=AppendVector(A=@n23.PrimaryR, B=@n23.PrimaryR), B=@n23.PrimaryB), B=@n23.PrimaryA)
  n25 = ScalarParameter() {defaultValue=1, group=Color, parameterName=Primary Color Intensity}
  n26 = CurveAtlasRowParameter(CurveTime=ComponentMask(None=TextureCoordinate()) {r=true}) {
   atlas=&TempAthenaCurveAtlas
   curve=&Curve_Pickaxe_Elastic_Fire_01
   defaultValue=14
  }
  n27 = Multiply(A=@n26.R) {constB=10}
  n28 = LinearInterpolate(
   A=Desaturation(None=Multiply(A=@n14, B=@n22), Fraction=ScalarParameter() {defaultValue=0.2, group=Color, parameterName=Secondary Color Desaturation})
   B=Desaturation(None=Multiply(A=@n24, B=@n25), Fraction=ScalarParameter() {defaultValue=0.15, group=Color, parameterName=Primary Color Desaturation})
   Alpha=LinearInterpolate(A=@n26.R, B=Saturate(None=@n27), Alpha=ScalarParameter() {group=Color, parameterName=Use_clear_boundaries})
  )
  n29 = LinearInterpolate(A=Multiply(A=Multiply(A=@n12.B, B=@n15), B=@n21), B=Multiply(A=Multiply(A=@n28, B=@n26.G), B=@n21), Alpha=Clamp(None=TwoSidedSign()))
  n30 = Panner(Coordinate=TextureCoordinate() {vTiling=2}) {speedY=0.5}
  n31 = Multiply(
   A=Desaturation(
    None=TextureSampleParameter2D(UVs=@n30) {group=Glitch, parameterName=Glitch, texture=/Game/Athena/UI/Challenges/Art/T_UI_DataChip_Noise}.RGB
    Fraction=ScalarParameter() {group=Glitch, parameterName=Des}
   )
   B=ScalarParameter() {defaultValue=20, group=Glitch, parameterName=Glitch_brightness}
  )
  n32 = VectorParameter() {group=Glitch, parameterName=Glitch_Tex_Color, defaultValue=<rgba 1 1 1 0>}
  n33 = LinearInterpolate(A=@n29, B=Multiply(A=Multiply(A=@n31, B=@n32.RGB), B=@n29), Alpha=ScalarParameter() {group=Glitch, parameterName=Use_Glitch_Color_Emissive})
  MaterialFunctionCall(Emissive=LinearInterpolate(A=@n29, B=Multiply(A=@n33, B=@n1.RGB), Alpha=@n6)) {materialFunction=&MF_FX_EmissiveAdjustments}
  ParticleColor()
  n34 = ScalarParameter() {defaultValue=1, group=WPO, parameterName=WPO_Intensity}
  n35 = TextureCoordinate()
  n36 = AppendVector(
   A=ScalarParameter() {defaultValue=2, group=WPO, parameterName=WPOTexture_Tile_U}
   B=ScalarParameter() {defaultValue=0.5, group=WPO, parameterName=WPOTexture_Tile_V}
  )
  n37 = ScalarParameter() {group=WPO, parameterName=WPOTexture_Pan_U}
  n38 = ScalarParameter() {defaultValue=-1, group=WPO, parameterName=WPOTexture_Pan_V}
  n39 = TextureSampleParameter2D(UVs=Panner(Coordinate=Multiply(A=@n35, B=@n36), Speed=AppendVector(A=@n37, B=@n38)) {speedY=0.3}) {group=WPO, parameterName=WPO_Texture, texture=&2ndInfectedMapLinear}
  n40 = Subtract(A=ComponentMask(None=TextureCoordinate()) {g=true}) {constB=0.5}
  n41 = Subtract(A=OneMinus(None=Abs(None=@n40))) {constB=0.5}
  Transform(None=Multiply(A=Multiply(A=@n34, B=@n39.RGB), B=Clamp(None=@n41)))
  VectorParameter() {group=Color, parameterName=Primary Color, defaultValue=<rgba 1 0.5 0 1>}
  VectorParameter() {group=Color, parameterName=Secondary Color, defaultValue=<rgba 1 0.05 0 1>}
 ]
}
