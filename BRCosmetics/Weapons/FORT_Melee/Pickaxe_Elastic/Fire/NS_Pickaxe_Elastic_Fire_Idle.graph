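# Idle Niagara fire effect for the Elastic Fire pickaxe, tinted by user Primary and Secondary colors.
# A persistent mesh, sprite and light flame is layered with flipbook soft fire, wrist glow and smoke, and vortex-swirled embers pulled along -X.
# Movement-spawned embers add a rising curl-noise trail.
# Two skeletal-position ribbons spiral via sine/cosine velocities split by spawn group, fading from primary to secondary color; the LightGlow emitter is disabled.
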NiagaraSystem /BRCosmetics/Weapons/FORT_Melee/Pickaxe_Elastic/Fire/NS_Pickaxe_Elastic_Fire_Idle
dependencies {usedDataInterfaces [&NiagaraDataInterfaceCurve], usedDynamicInputs [], usedModules [&SystemState]}
emitters {
 Decal_Glow_Wrist {
  emitterUpdateScript [
   &EmitterState()
   &SpawnBurst_Instantaneous(Spawn Count=1, Use Spawn Probability=false)
  ]
  particleSpawnScript [
   /Game/Effects/Niagara/Modules/Partition/FN_ParticlePartition_v2/FN_ParticlePartition_v2()
   /Niagara/Modules/Spawn/Initialization/InitializeParticle(
    Color=$User.Primary Color
    Module.Write Color=true
    Module.Write Lifetime=true
    Module.Write Mass=true
    Module.Write Material Random=true
    Module.Write Position=false
    Module.Write Scale=false
    Module.Write SpriteRotation=false
    Module.Write SpriteSize=true
    Position=&SimulationPosition()
    Sprite Size=&Vector2DFromFloat(Value=25)
   )
   /Game/Effects/Niagara/Modules/BoneFlow/FN_Initialize_BoneFlow_Attributes_v2()
   &SystemLocation(MaskSpawnGroup=false, Offset=<10 0 0>)
  ]
  particleUpdateScript [
   &ParticleState()
   &DynamicMaterialParameters(
    Index 0 Param 1=1.5
    Module.Param0WriteEnabled=true
    Module.Param1WriteEnabled=false
    Module.Param2WriteEnabled=false
    Module.Param3WriteEnabled=false
   ) {module=DynamicMaterialParameters001}
   &ScaleColor(Scale RGB=&VectorFromFloat(Value=2), ScaleA=false, ScaleRGB=true)
   &CameraOffset(Camera Offset Amount=20)
  ]
  rendererClasses [&NiagaraSpriteRendererProperties]
  renderers [{rendererClass=&NiagaraSpriteRendererProperties}]
 }
 Decal_Smoke_Wrist {
  emitterUpdateScript [
   &EmitterState()
   &SpawnRate(SpawnRate=3, Use Spawn Probability=false)
  ]
  particleSpawnScript [
   &FN_ParticlePartition_v2()
   &InitializeParticle(
    Color=$User.Primary Color
    Lifetime=&UniformRangedFloat(Minimum=1, Maximum=1.5)
    Module.Write Color=true
    Module.Write Lifetime=true
    Module.Write Mass=true
    Module.Write Material Random=true
    Module.Write Position=false
    Module.Write Scale=false
    Module.Write SpriteRotation=true
    Module.Write SpriteSize=true
    Position=&SimulationPosition()
    Sprite Rotation=&UniformRangedFloat(Minimum=180, Maximum=-180)
    Sprite Size=&Vector2DFromFloat(Value=&UniformRangedFloat(Minimum=16, Maximum=20))
   )
   &FN_Initialize_BoneFlow_Attributes_v2()
   &SystemLocation(MaskSpawnGroup=false, Offset=<20 0 0>)
   &AddVelocity(Velocity=&UniformRangedVector(Minimum=<-10 0 0>, Maximum=<-15 0 0>))
  ]
  particleUpdateScript [
   &ParticleState()
   &ScaleColor(
    Scale Alpha=&FloatFromCurve(
     FloatCurve=DataInterface() {propertyValues {Curve {keys [
         {interpMode=RCIM_Cubic, tangentMode=RCTM_Break, time=0, value=0}
         {interpMode=RCIM_Cubic, tangentMode=RCTM_User, time=0.4, value=0.9932}
         {interpMode=RCIM_Cubic, tangentMode=RCTM_Break, time=1, value=0}
        ]}}}
     Scale Curve=3
    )
    Scale RGB=&VectorFromFloat(Value=8)
    ScaleA=true
    ScaleRGB=true
   )
   &CameraOffset(Camera Offset Amount=20)
   &SolveForcesAndVelocity(Clamp Velocity=false, Limit Acceleration=false)
   &DynamicMaterialParameters(
    Index 0 Param 1=&FloatFromCurve(FloatCurve=DataInterface() {propertyValues {Curve {keys [{interpMode=RCIM_Cubic, leaveTangent=2.3044, tangentMode=RCTM_Break, time=0, value=0.25}, <at 1 1>]}}})
    Index 0 Param 2=0.5
    Module.Param0WriteEnabled=true
    Module.Param1WriteEnabled=true
    Module.Param2WriteEnabled=true
    Module.Param3WriteEnabled=true
   )
  ]
  rendererClasses [&NiagaraSpriteRendererProperties]
  renderers [{rendererClass=&NiagaraSpriteRendererProperties}]
 }
 Embers_LocalSpace {
  emitterUpdateScript [
   &EmitterState(Enable Distance Culling=true, MaxDistance=2000, Module.bUseMaxDistance=true, Module.bUseMinDistance=false, Scalability Mode=::NewEnumerator1)
   &SpawnRate(SpawnRate=5, Use Spawn Probability=false)
  ]
  particleSpawnScript [
   &InitializeParticle(
    Color=$User.Primary Color
    Lifetime=&UniformRangedFloat(Minimum=0.5, Maximum=1.5)
    Module.Write Color=true
    Module.Write Lifetime=true
    Module.Write Mass=true
    Module.Write Material Random=true
    Module.Write Position=false
    Module.Write Scale=false
    Module.Write SpriteRotation=true
    Module.Write SpriteSize=true
    Position=&SimulationPosition()
    Sprite Size=&UniformRangedVector2D(Minimum=<1.15 2.3>, Maximum=<2 3>)
   )
   &FN_Initialize_BoneFlow_Attributes_v2()
   &SystemLocation(MaskSpawnGroup=false)
   &CylinderLocation(
    Cylinder Height=6
    Cylinder Radius=6
    MaskSpawnGroup=false
    Module.Enable Lathe Profile=false
    Module.Override Randomness=false
    Module.Surface Only=false
    Offset=<12 0 0>
    Orientation Axis:ENiagaraOrientationAxis=::XAxis
   )
   &AddVelocity(Velocity=&Multiply_VectorByFloat(Vector=&Multiply_Vector(A=$Output.CylinderLocation.CylinderNormal, B=<0 1 1>), Float=50)) {module=AddVelocity002}
   &AddVelocity(Velocity=&UniformRangedVector(Minimum=<-20 0 0>, Maximum=<-40 0 0>))
  ]
  particleUpdateScript [
   &ParticleState()
   /Niagara/Modules/Update/Velocity/VortexVelocity(Influence Falloff Radius=50, Module.Use Influence Falloff=false, Velocity Amount=-40, Vortex Axis:Vector3f=<1 0 0>)
   /Niagara/Modules/Update/Forces/PointAttractionForce(
    Attraction Radius=40
    AttractionStrength=300
    AttractorPosition:Vector3f=&Add_Vector(A=/Niagara/DynamicInputs/Helpers/SimulationPosition(), B=<-40 0 0>)
    Falloff Exponent=0.4
    Module.Kill Within Radius=false
    Module.Use Falloff=true
   )
   &SolveForcesAndVelocity(Clamp Velocity=false, Limit Acceleration=false)
   &ScaleColor(
    Scale Alpha=&FloatFromCurve(FloatCurve=DataInterface() {propertyValues {Curve {keys [
         {interpMode=RCIM_Cubic, tangentMode=RCTM_Break, time=0, value=0}
         <at 0.4 1>
         {interpMode=RCIM_Cubic, tangentMode=RCTM_User, time=0.65, value=1}
         {interpMode=RCIM_Cubic, tangentMode=RCTM_Break, time=1, value=0}
        ]}}})
    Scale RGB=<20 20 20>
    ScaleA=true
    ScaleRGB=true
   )
   &CameraOffset(Camera Offset Amount=10)
   &ScaleSpriteSize(Scale Factor=&Vector2DFromFloat(Value=&FloatFromCurve(FloatCurve=DataInterface() {propertyValues {Curve {keys [
          {arriveTangent=4.4129, interpMode=RCIM_Cubic, leaveTangent=4.4129, tangentMode=RCTM_User, time=0, value=0.5}
          <at 0.1967 1>
          <at 1 0 RCIM_Cubic>
         ]}}})))
  ]
  rendererClasses [&NiagaraSpriteRendererProperties]
  renderers [{rendererClass=&NiagaraSpriteRendererProperties}]
 }
 Embers_WorldSpace {
  emitterUpdateScript [
   &EmitterState(Enable Distance Culling=true, MaxDistance=2000, Module.bUseMaxDistance=true, Module.bUseMinDistance=false, Scalability Mode=::NewEnumerator1)
   &SpawnPerUnit(Module.Use Max Movement Threshold=true, Module.Use Movement Tolerance=false, Spawn Per Unit=20, Use Spawn Probability=false)
  ]
  particleSpawnScript [
   &InitializeParticle(
    Color=$User.Primary Color
    Lifetime=&UniformRangedFloat(Minimum=0.15, Maximum=0.4)
    Module.Write Color=true
    Module.Write Lifetime=true
    Module.Write Mass=true
    Module.Write Material Random=true
    Module.Write Position=false
    Module.Write Scale=false
    Module.Write SpriteRotation=true
    Module.Write SpriteSize=true
    Position=/Niagara/Functions/Localspace/SimulationPosition()
    Sprite Rotation=&UniformRangedFloat(Minimum=0, Maximum=360)
    Sprite Size=&UniformRangedVector2D(Minimum=<1 1>, Maximum=<3 3>)
   )
   &FN_Initialize_BoneFlow_Attributes_v2()
   &SystemLocation(MaskSpawnGroup=false, Offset=<-10 0 0>)
   &CylinderLocation(
    Cylinder Height=20
    Cylinder Radius=6
    MaskSpawnGroup=false
    Module.Enable Lathe Profile=false
    Module.Override Randomness=false
    Module.Surface Only=false
    Orientation Axis:ENiagaraOrientationAxis=::XAxis
   )
   &AddVelocity(Velocity=&Multiply_VectorByFloat(Vector=&Multiply_Vector(A=$Output.CylinderLocation.CylinderNormal, B=<0 1 1>), Float=50)) {module=AddVelocity002}
   &AddVelocity(Velocity=&UniformRangedVector(Minimum=<-20 0 0>, Maximum=<-40 0 0>))
  ]
  particleUpdateScript [
   &ParticleState()
   &AccelerationForce(Acceleration=<0 0 200>)
   /Niagara/Modules/Update/Forces/CurlNoiseForce(
    Noise Frequency=0.04
    Noise Strength=&FloatFromCurve(FloatCurve=DataInterface() {propertyValues {Curve {keys [<at 0 1>,<at 1 0>]}}}, Scale Curve=300)
    Pan Noise=true
    Pan Noise Field=<0 0 1>
   )
   &InheritVelocity(Module.Limit Speed=true, Velocity Scale=<0.05 0.05 0.05>)
   &SolveForcesAndVelocity(Clamp Velocity=false, Limit Acceleration=false)
   &ScaleColor(
    Scale Alpha=&FloatFromCurve(FloatCurve=DataInterface() {propertyValues {Curve {keys [
         {interpMode=RCIM_Cubic, tangentMode=RCTM_Break, time=0, value=0}
         <at 0.4 1>
         {interpMode=RCIM_Cubic, tangentMode=RCTM_User, time=0.65, value=1}
         {interpMode=RCIM_Cubic, tangentMode=RCTM_Break, time=1, value=0}
        ]}}})
    Scale RGB=<20 20 20>
    ScaleA=true
    ScaleRGB=true
   )
   &CameraOffset(Camera Offset Amount=10)
   &ScaleSpriteSize(Scale Factor=&Vector2DFromFloat(Value=&FloatFromCurve(FloatCurve=DataInterface() {propertyValues {Curve {keys [
          {arriveTangent=4.4129, interpMode=RCIM_Cubic, leaveTangent=4.4129, tangentMode=RCTM_User, time=0, value=0.5}
          <at 0.1967 1>
          <at 1 0 RCIM_Cubic>
         ]}}})))
  ]
  rendererClasses [&NiagaraSpriteRendererProperties]
  renderers [{rendererClass=&NiagaraSpriteRendererProperties}]
 }
 FN_Contrails_RibbonWde {
  emitterUpdateScript [
   &EmitterState()
   &FN_SpawnRibbon_v2(Number of Ribbons=2, Spawn Mode=::NewEnumerator0, SpawnRate=60)
  ]
  particleSpawnScript [
   &FN_InitializeRibbon_v3(Enable Skeletal Position=true, Lifetime=1, Offset Axis:ENiagaraOrientationAxis=::XAxis, Ribbon Width=20)
   &SystemLocation(MaskSpawnGroup=false, Offset=<13 0 0>)
   &AddVelocity(Velocity=&MakeVector(
     X=0
     Y=&MaskFloatBySpawnGroup(
      Masked Float=&Sine(
       Normalized Angle=&Multiply_Float(A=$Emitter.Age, B=-1)
       Period=0.6
       Scale=&FloatFromCurve(
        FloatCurve=DataInterface() {propertyValues {Curve {keys [<at 0 1 RCIM_Cubic>, {arriveTangent=-2.1595, interpMode=RCIM_Cubic, leaveTangent=-2.1595, tangentMode=RCTM_User, time=1, value=0}]}}}
        CurveIndex=&NormalizeDistanceRange(
         Start Position:NiagaraPosition=/Niagara/DynamicInputs/Helpers/SimulationPosition()
         End Position:NiagaraPosition=<0 0 0>
         Distance:NiagaraFloat=1.5
         Invert Normalized Range:NiagaraBool=false
        )
        Scale Curve=45
       )
       Bias=1
      )
      Passthrough Float=&Cosine(
       Normalized Angle=&Multiply_Float(A=$Emitter.Age, B=-1)
       Period=0.6
       Scale=&FloatFromCurve(
        FloatCurve=DataInterface() {propertyValues {Curve {keys [<at 0 1 RCIM_Cubic>, {arriveTangent=-2.1595, interpMode=RCIM_Cubic, leaveTangent=-2.1595, tangentMode=RCTM_User, time=1, value=0}]}}}
        CurveIndex=&NormalizeDistanceRange(
         Start Position:NiagaraPosition=&SimulationPosition()
         End Position:NiagaraPosition=<0 0 0>
         Distance:NiagaraFloat=1.5
         Invert Normalized Range:NiagaraBool=false
        )
        Scale Curve=45
       )
       Bias=1
      )
     )
     Z=&MaskFloatBySpawnGroup(
      Masked Float=&Cosine(
       Normalized Angle=&Multiply_Float(A=$Emitter.Age, B=-1)
       Period=0.6
       Scale=&FloatFromCurve(
        FloatCurve=DataInterface() {propertyValues {Curve {keys [<at 0 1 RCIM_Cubic>, {arriveTangent=-2.1595, interpMode=RCIM_Cubic, leaveTangent=-2.1595, tangentMode=RCTM_User, time=1, value=0}]}}}
        CurveIndex=&NormalizeDistanceRange(
         Start Position:NiagaraPosition=&SimulationPosition()
         End Position:NiagaraPosition=<0 0 0>
         Distance:NiagaraFloat=1.5
         Invert Normalized Range:NiagaraBool=false
        )
        Scale Curve=45
       )
       Bias=1
      )
      Passthrough Float=&Sine(
       Normalized Angle=&Multiply_Float(A=$Emitter.Age, B=-1)
       Period=0.6
       Scale=&FloatFromCurve(
        FloatCurve=DataInterface() {propertyValues {Curve {keys [<at 0 1 RCIM_Cubic>, {arriveTangent=-2.1595, interpMode=RCIM_Cubic, leaveTangent=-2.1595, tangentMode=RCTM_User, time=1, value=0}]}}}
        CurveIndex=&NormalizeDistanceRange(
         Start Position:NiagaraPosition=&SimulationPosition()
         End Position:NiagaraPosition=<0 0 0>
         Distance:NiagaraFloat=1.5
         Invert Normalized Range:NiagaraBool=false
        )
        Scale Curve=45
       )
       Bias=1
      )
     )
    ))
  ]
  particleUpdateScript [
   &ScaleRibbonWidth(Ribbon Width Scale=&FloatFromCurve(FloatCurve=DataInterface() {propertyValues {Curve {keys [<at 0 0.2 RCIM_Cubic>, {arriveTangent=-0.2, interpMode=RCIM_Cubic, leaveTangent=-0.2, time=0.1953, value=0.6379}, <at 1 0 RCIM_Cubic>]}}}))
   &Material(
    Base Texture Time:NiagaraFloat=&Multiply_Float(A=$Emitter.Age, B=&Lerp_Float(A=-0.4, B=-100, Alpha=$Particles.IntegratedDive))
    PreMult=0
   )
   &PointAttractionForce(
    Attraction Radius=40
    AttractionStrength=110
    AttractorPosition:Vector3f=&Add_Vector(A=&SimulationPosition(), B=<-15 0 0>)
    Falloff Exponent=0.18
    Module.Kill Within Radius=false
    Module.Use Falloff=true
   )
   &AddVelocity() {module=AddVelocity001}
   &SolveForcesAndVelocity(Clamp Velocity=false, Limit Acceleration=false)
   &FN_UpdateRibbon(
    Enable Skeletal Position=true
    Initial Color:LinearColor=$Particles.Initial.Color
    Initial RibbonWidth:NiagaraFloat=$Particles.Initial.RibbonWidth
   )
   /SetVariables(Particles.SpriteFacing=HlslExpression() {hlslExpression="dot(Particles.RibbonFacing,  float3(0,0,1))"}) {isSetParameters=true}
   &ScaleColor(
    Scale Alpha=&FloatFromCurve(FloatCurve=DataInterface() {propertyValues {Curve {keys [<at 0 1 RCIM_Cubic>,<at 1 0 RCIM_Cubic>]}}})
    Scale RGB=&Multiply_VectorByFloat(
     Vector=&Lerp_Vector(
      A=&MakeVectorFromLinearColorRGB(Color=$User.Primary Color)
      B=&MakeVectorFromLinearColorRGB(Color=$User.Secondary Color)
      Alpha=$Particles.NormalizedAge
     )
     Float=&FloatFromCurve(
      FloatCurve=DataInterface() {propertyValues {Curve {keys [
          <at 0 1 RCIM_Cubic>
          {arriveTangent=-1.7902, interpMode=RCIM_Cubic, leaveTangent=-1.7902, time=0.2294, value=0.5134}
          {arriveTangent=-0.6663, interpMode=RCIM_Cubic, leaveTangent=-0.6663, time=0.5108, value=0.0857}
          <at 1 0 RCIM_Cubic>
         ]}}}
      Scale Curve=10
     )
    )
    ScaleA=true
    ScaleRGB=true
   )
  ]
  rendererClasses [&NiagaraRibbonRendererProperties]
  renderers [{rendererClass=&NiagaraRibbonRendererProperties}]
 }
 Flame_MAIN {
  emitterUpdateScript [
   &EmitterState()
   &SpawnBurst_Instantaneous(Spawn Count=1, Use Spawn Probability=false)
  ]
  particleSpawnScript [
   &InitializeParticle(
    Color=$User.Primary Color
    Mesh Scale=&Multiply_Vector(A=<2.8 2.6 2.6>, B=&VectorFromFloat(Value=1))
    Module.Write Color=true
    Module.Write Lifetime=true
    Module.Write Mass=true
    Module.Write Material Random=false
    Module.Write Position=false
    Module.Write Scale=true
    Module.Write SpriteRotation=false
    Module.Write SpriteSize=true
    Position=/Niagara/Functions/Localspace/SimulationPosition()
    Sprite Size=<40 40>
   )
   /SetVariables(Particles.DynMatParm_2:Vector4f=&MakeVector4(X=1, Y=0, Z=0, W=0)) {isSetParameters=true}
   /SetVariables(Particles.PosOffset_2:NiagaraPosition=$Particles.Position) {isSetParameters=true}
   &SystemLocation(MaskSpawnGroup=false, Offset=<12 0 0>)
  ]
  particleUpdateScript [
   &ParticleState()
   &CameraOffset(Camera Offset Amount=15)
   /SetVariables() {isSetParameters=true}
   /Niagara/Modules/Update/Orientation/V2/UpdateMeshOrientation(Rotation Vector:Vector3f=<1 0 0>)
   &self:DynParam_ColorPicker(Primary Color:LinearColor=$User.Primary Color, Secondary Color:LinearColor=$User.Secondary Color) {module=DynParam_ColorPicker}
   &ScaleColor(Scale Alpha=0.4, Scale RGB=&VectorFromFloat(Value=0.5), ScaleA=true, ScaleRGB=true)
   &FN_Light_v2(
    Light Exponent=3
    Light Radius=25
    LightEnabled=true
    Position Offset:Vector3f=&Add_Vector(A=$Particles.PosOffset_2, B=<-12 0 -10>)
    Scale Color=75
   )
  ]
  rendererClasses [
   &NiagaraMeshRendererProperties
   &NiagaraSpriteRendererProperties
   &NiagaraLightRendererProperties
  ]
  renderers [
   {rendererClass=&NiagaraMeshRendererProperties, rendererIndex=0}
   {rendererClass=&NiagaraSpriteRendererProperties, rendererIndex=1}
   {rendererClass=&NiagaraLightRendererProperties, rendererIndex=2}
  ]
 }
 LightGlow {
  bEnabled=false
  emitterUpdateScript [
   &EmitterState(Inactive Response=::NewEnumerator1, Life Cycle Mode=::NewEnumerator1, Loop Duration=1, UseLoopDelay=false)
   &SpawnBurst_Instantaneous(Spawn Count=1, Use Spawn Probability=false)
  ]
  particleSpawnScript [
   &InitializeParticle(
    Color=$User.Primary Color
    Module.Write Color=true
    Module.Write Lifetime=true
    Module.Write Mass=true
    Module.Write Material Random=true
    Module.Write Position=false
    Module.Write Scale=false
    Module.Write SpriteRotation=false
    Module.Write SpriteSize=true
    Position=&SimulationPosition()
    Sprite Size=&Vector2DFromFloat(Value=150)
   )
   &DynamicMaterialParameters(
    Index 0 Param 1=1
    Module.Param0WriteEnabled=true
    Module.Param1WriteEnabled=true
    Module.Param2WriteEnabled=true
    Module.Param3WriteEnabled=true
   )
   &SystemLocation(MaskSpawnGroup=false)
  ]
  particleUpdateScript [
   &CameraOffset(Camera Offset Amount=15)
   &ParticleState()
   &ScaleColor(Scale Alpha=0.1, Scale RGB=&VectorFromFloat(Value=1), ScaleA=true, ScaleRGB=true)
   &FN_Light_v2(Light Exponent=3, Light Radius=25, LightEnabled=true, Position Offset:Vector3f=<0 5 -10>, Scale Color=75)
  ]
  rendererClasses [&NiagaraSpriteRendererProperties, &NiagaraLightRendererProperties]
  renderers [
   {rendererClass=&NiagaraSpriteRendererProperties}
   {rendererClass=&NiagaraLightRendererProperties, rendererIndex=1}
  ]
 }
 SoftFire_Inside {
  emitterUpdateScript [
   &EmitterState()
   &SpawnRate(SpawnRate=4, Use Spawn Probability=false)
  ]
  particleSpawnScript [
   &FN_ParticlePartition_v2()
   &InitializeParticle(
    Color=$User.Primary Color
    Lifetime=1.4
    Module.Write Color=true
    Module.Write Lifetime=true
    Module.Write Mass=true
    Module.Write Material Random=true
    Module.Write Position=false
    Module.Write Scale=false
    Module.Write SpriteRotation=false
    Module.Write SpriteSize=true
    Position=&SimulationPosition()
    Sprite Size=<25 50>
   )
   &FN_Initialize_BoneFlow_Attributes_v2()
   &SystemLocation(MaskSpawnGroup=false, Offset=<20 0 0>)
   &AddVelocity(Velocity=&UniformRangedVector(Minimum=<-35 0 0>, Maximum=<-50 0 0>))
  ]
  particleUpdateScript [
   &ParticleState()
   &SolveForcesAndVelocity(Clamp Velocity=false, Limit Acceleration=false)
   &ScaleColor(
    Scale Alpha=&FloatFromCurve(FloatCurve=DataInterface() {propertyValues {Curve {keys [
         {interpMode=RCIM_Cubic, leaveTangent=6.4595, tangentMode=RCTM_Break, time=0, value=0}
         <at 0.3081 0.9779>
         {interpMode=RCIM_Cubic, tangentMode=RCTM_User, time=0.625, value=0.9901}
         {interpMode=RCIM_Cubic, tangentMode=RCTM_Break, time=1, value=0}
        ]}}})
    Scale RGB=&VectorFromFloat(Value=&FloatFromCurve(FloatCurve=DataInterface() {propertyValues {Curve {keys [<at 0 1>,<at 1 0>]}}}, Scale Curve=10))
    ScaleA=true
    ScaleRGB=true
   )
   &CameraOffset(Camera Offset Amount=10)
   &ScaleSpriteSize(Scale Factor=&Vector2DFromFloat(Value=&FloatFromCurve(FloatCurve=DataInterface() {propertyValues {Curve {keys [
          {arriveTangent=8.0919, interpMode=RCIM_Cubic, leaveTangent=2.0645, tangentMode=RCTM_User, time=-0.0094, value=0.0139}
          {arriveTangent=1.0622, interpMode=RCIM_Cubic, leaveTangent=1.0622, time=0.2785, value=0.5542}
          {arriveTangent=-0.4825, interpMode=RCIM_Cubic, leaveTangent=-0.4825, time=0.4835, value=0.5375}
          {arriveTangent=-1.2463, interpMode=RCIM_Cubic, leaveTangent=-1.2463, tangentMode=RCTM_User, time=0.9924, value=0.2098}
         ]}}})))
   /Niagara/Modules/Update/SubUV/SubUVAnimation(
    Number Of Frames=64
    SubUV Lookup Index=&Multiply_Float(A=&Add_Float(A=$Particles.NormalizedAge, B=$Particles.RandomStartingFrame), B=0.6)
   )
   &DynamicMaterialParameters(
    Index 0 Param 1=&FloatFromCurve(FloatCurve=DataInterface() {propertyValues {Curve {keys [<at 0 0.25 RCIM_Cubic>,<at 1 -0.1>]}}}, CurveIndex=$Particles.NormalizedLife)
    Module.Param0WriteEnabled=true
    Module.Param1WriteEnabled=false
    Module.Param2WriteEnabled=false
    Module.Param3WriteEnabled=false
   )
  ]
  rendererClasses [&NiagaraSpriteRendererProperties]
  renderers [{rendererClass=&NiagaraSpriteRendererProperties}]
 }
 SoftFire_Outside {
  emitterUpdateScript [
   &EmitterState()
   &SpawnRate(SpawnRate=5, Use Spawn Probability=false)
  ]
  particleSpawnScript [
   &FN_ParticlePartition_v2()
   &InitializeParticle(
    Color=$User.Primary Color
    Lifetime=2.5
    Module.Write Color=true
    Module.Write Lifetime=true
    Module.Write Mass=true
    Module.Write Material Random=true
    Module.Write Position=false
    Module.Write Scale=false
    Module.Write SpriteRotation=true
    Module.Write SpriteSize=true
    Position=&SimulationPosition()
    Sprite Rotation=&UniformRangedFloat(Minimum=0, Maximum=360)
    Sprite Size=&UniformRangedVector2D(Minimum=<25 25>, Maximum=<35 35>)
   )
   &FN_Initialize_BoneFlow_Attributes_v2()
   &SystemLocation(MaskSpawnGroup=false, Offset=<20 0 0>)
   &SphereLocation(
    MaskSpawnGroup=false
    Override Seed=false
    Sphere Radius=6
    Surface Only=false
    UseNonUniformScale=false
    UseOffset=false
    UseSphereOrientation=false
   )
   &AddVelocity(Velocity=&UniformRangedVector(Minimum=<-20 0 0>, Maximum=<-15 0 0>))
   /SetVariables(Particles.RandomStartingFrame:NiagaraFloat=&UniformRangedFloat(Minimum=0, Maximum=1)) {isSetParameters=true}
   /SetVariables(Particles.InitialRotationRate:NiagaraFloat=&UniformRangedFloat(Minimum=-60, Maximum=0)) {isSetParameters=true}
  ]
  particleUpdateScript [
   &ParticleState()
   &VortexVelocity(Module.Use Influence Falloff=false, Velocity Amount=30, Vortex Axis:Vector3f=<1 0 0>)
   &PointAttractionForce(
    Attraction Radius=65
    AttractionStrength=75
    AttractorPosition:Vector3f=&Add_Vector(A=/Niagara/DynamicInputs/Helpers/SimulationPosition(), B=<-40 0 0>)
    Module.Kill Within Radius=false
    Module.Use Falloff=true
   )
   &SolveForcesAndVelocity(Clamp Velocity=false, Limit Acceleration=false)
   &ScaleColor(
    Scale Alpha=&FloatFromCurve(
     FloatCurve=DataInterface() {propertyValues {Curve {keys [
         {interpMode=RCIM_Cubic, leaveTangent=8.9046, tangentMode=RCTM_Break, time=0, value=0}
         <at 0.1192 0.9901>
         {interpMode=RCIM_Cubic, tangentMode=RCTM_User, time=0.65, value=1}
         {interpMode=RCIM_Cubic, tangentMode=RCTM_Break, time=1, value=0}
        ]}}}
     Scale Curve=0.2
    )
    Scale RGB=<6 6 6>
    ScaleA=true
    ScaleRGB=true
   )
   &CameraOffset(Camera Offset Amount=&FloatFromCurve(FloatCurve=DataInterface() {propertyValues {Curve {keys [<at 0.0099 23.4541>,<at 0.1387 27.7338>,<at 0.612 17.5534>,<at 0.7977 21.7682>,<at 1 20>]}}}))
   &ScaleSpriteSize(Scale Factor=&Vector2DFromFloat(Value=&FloatFromCurve(FloatCurve=DataInterface() {propertyValues {Curve {keys [
          {arriveTangent=4.4129, interpMode=RCIM_Cubic, leaveTangent=4.4129, tangentMode=RCTM_User, time=0, value=0.5}
          <at 0.1967 1>
          <at 1 0 RCIM_Cubic>
         ]}}})))
   &SubUVAnimation(
    Number Of Frames=64
    SubUV Lookup Index=&Multiply_Float(A=&Add_Float(A=$Particles.NormalizedAge, B=$Particles.RandomStartingFrame), B=0.6)
   )
   &SpriteRotationRate(Rotation Rate=$Particles.InitialRotationRate)
   &DynamicMaterialParameters(
    Index 0 Param 1=&FloatFromCurve(FloatCurve=DataInterface() {propertyValues {Curve {keys [<at 0 0.25 RCIM_Cubic>,<at 1 -0.1>]}}}, CurveIndex=$Particles.NormalizedLife)
    Module.Param0WriteEnabled=true
    Module.Param1WriteEnabled=false
    Module.Param2WriteEnabled=false
    Module.Param3WriteEnabled=false
   )
  ]
  rendererClasses [&NiagaraSpriteRendererProperties]
  renderers [{rendererClass=&NiagaraSpriteRendererProperties}]
 }
}
summary {
 systemName=NS_Pickaxe_Elastic_Fire_Idle
 emitters [
  {emitterName=SoftFire_Outside, rendererClasses [&NiagaraSpriteRendererProperties]}
  {
   emitterName=Flame_MAIN
   rendererClasses [
    &NiagaraMeshRendererProperties
    &NiagaraSpriteRendererProperties
    &NiagaraLightRendererProperties
   ]
  }
  {
   bEnabled=false
   emitterName=LightGlow
   rendererClasses [&NiagaraSpriteRendererProperties, &NiagaraLightRendererProperties]
  }
  {emitterName=Decal_Glow_Wrist, rendererClasses [&NiagaraSpriteRendererProperties]}
  {emitterName=Decal_Smoke_Wrist, rendererClasses [&NiagaraSpriteRendererProperties]}
  {emitterName=Embers_LocalSpace, rendererClasses [&NiagaraSpriteRendererProperties]}
  {emitterName=SoftFire_Inside, rendererClasses [&NiagaraSpriteRendererProperties]}
  {emitterName=FN_Contrails_RibbonWde, rendererClasses [&NiagaraRibbonRendererProperties]}
  {emitterName=Embers_WorldSpace, rendererClasses [&NiagaraSpriteRendererProperties]}
 ]
 userVariables [
  User.Primary Color {type=LinearColor, description="", defaultValue {struct=LinearColor, value=<rgba 0 0.1333 1 1>}}
  User.RibbonIDOffset {type=NiagaraInt32}
  User.Secondary Color {type=LinearColor, description="", defaultValue {struct=LinearColor, value=<rgba 0.5 0 1 1>}}
  User.SkeletalMesh {
   type=NiagaraDataInterfaceSkeletalMesh
   defaultValue {
    struct=NiagaraExt_VariableValue_DataInterface
    value {dataInterface=&self:NiagaraDataInterfaceSkeletalMesh_0}
   }
  }
 ]
}
